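# Physics asset for the HerbHutch gadget prop's skeletal mesh, a creature rig with wings, four legs, neck and tail.
# Single capsule bodies cover the center of gravity, both wings, the four leg clavicles, two neck joints and the tail base, plus a tiny one on main.
# Wings and center body constrain to main; legs, tail and lower neck to the center body; upper neck to lower neck.
# All constraints share uniform swing and twist limit angles.
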PhysicsAsset /BRCosmetics/Animation/Props/Gadgets_HerbHutch/Meshes/HerbHutch_Gadget_PhysicsAsset
bodies {
 JNT_CoG [{shapeType=Capsule, length=20.3075, radius=11.0137, center=<-0.1511 -1.764 -0.2732>, rotation=<rot 0 -175.1434 -89.6713>}]
 JNT_L_Wing_01 [{shapeType=Capsule, length=18.4069, radius=7.1405, center=<18.0159 -3.7159 -0.251>, rotation=<rot 0 -98.9059 -90>}]
 JNT_Lf_BackLeg_Clav [{shapeType=Capsule, length=8.105, radius=5.7482, center=<-0.567 6.6659 -1.4919>, rotation=<rot 0 -165.5218 -45.9815>}]
 JNT_Lf_FrontLeg_Clav [{shapeType=Capsule, length=9.3046, radius=4.7507, center=<0.212 5.3613 0.6315>, rotation=<rot 0 0.5491 -65.2289>}]
 JNT_Neck_1 [{shapeType=Capsule, length=10.1955, radius=8.7122, center=<0.0814 3.463 -2.5985>, rotation=<rot 13.6298 -90 4.0639>}]
 JNT_Neck_4 [{shapeType=Capsule, length=13.9529, radius=11.1317, center=<0 10.1048 -0.0409>, rotation=<rot 35.4017 -90 0.0009>}]
 JNT_R_Wing_01 [{shapeType=Capsule, length=18.4069, radius=7.1405, center=<-18.0159 3.7159 0.251>, rotation=<rot 0 -98.9059 -90>}]
 JNT_Rt_BackLeg_Clav [{shapeType=Capsule, length=8.3321, radius=5.7715, center=<0.581 -6.5061 1.0943>, rotation=<rot 0 -174.9194 -50.3284>}]
 JNT_Rt_FrontLeg_Clav [{shapeType=Capsule, length=9.1958, radius=5.4959, center=<-0.0132 -5.3103 -0.9688>, rotation=<rot 0 9.7152 -56.4371>}]
 JNT_Tail_1 [{shapeType=Capsule, length=7.735, radius=7.6462, center=<-0.124 -6.344 0.5532>, rotation=<rot 4.2357 91.1972 -90>}]
 main [{shapeType=Capsule, length=0.505, radius=0.505, rotation=<rot 0 -90 0>}]
}
constraints [
 {bone1Name=JNT_L_Wing_01, bone2Name=main, swing1LimitDegrees=45, swing2LimitDegrees=45, twistLimitDegrees=45}
 {bone1Name=JNT_R_Wing_01, bone2Name=main, swing1LimitDegrees=45, swing2LimitDegrees=45, twistLimitDegrees=45}
 {bone1Name=JNT_CoG, bone2Name=main, swing1LimitDegrees=45, swing2LimitDegrees=45, twistLimitDegrees=45}
 {bone1Name=JNT_Lf_BackLeg_Clav, bone2Name=JNT_CoG, swing1LimitDegrees=45, swing2LimitDegrees=45, twistLimitDegrees=45}
 {bone1Name=JNT_Rt_BackLeg_Clav, bone2Name=JNT_CoG, swing1LimitDegrees=45, swing2LimitDegrees=45, twistLimitDegrees=45}
 {bone1Name=JNT_Tail_1, bone2Name=JNT_CoG, swing1LimitDegrees=45, swing2LimitDegrees=45, twistLimitDegrees=45}
 {bone1Name=JNT_Lf_FrontLeg_Clav, bone2Name=JNT_CoG, swing1LimitDegrees=45, swing2LimitDegrees=45, twistLimitDegrees=45}
 {bone1Name=JNT_Rt_FrontLeg_Clav, bone2Name=JNT_CoG, swing1LimitDegrees=45, swing2LimitDegrees=45, twistLimitDegrees=45}
 {bone1Name=JNT_Neck_1, bone2Name=JNT_CoG, swing1LimitDegrees=45, swing2LimitDegrees=45, twistLimitDegrees=45}
 {bone1Name=JNT_Neck_4, bone2Name=JNT_Neck_1, swing1LimitDegrees=45, swing2LimitDegrees=45, twistLimitDegrees=45}
]
